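# Material function that produces hard-edged world-space triplanar UVs, choosing a single planar projection per pixel instead of blending three.
# World position with flipped Z is divided by UV Scale, then lerped between XY, XZ and YZ planes; the reroutes feeding the lerps show no declaration link.
# Absolute vertex normal components are compared and rounded into a one-hot Triplanar Mask.
# A Seam Mask marks projection boundaries, tightened by a Seam Sharpness exponent.
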MaterialFunction /Game/Athena/Environments/MaterialFunctions/MF_TriplanarFlat
graph {nodes [
  n1 = FunctionInput() {bUsePreviewValueAsDefault=true, inputName=UV Scale, inputType=FunctionInput_Scalar, previewValue=<650 650 650 650>}
  n2 = Divide(A=Multiply(A=WorldPosition().XYZ, B=Constant3Vector() {constant=<rgba 1 1 -1 0>}), B=@n1)
  n3 = ComponentMask(None=@n2) {g=true, r=true}
  n4 = ComponentMask(None=@n2) {b=true, r=true}
  n5 = NamedRerouteUsage()
  FunctionOutput(None=LinearInterpolate(A=LinearInterpolate(A=@n3, B=@n4, Alpha=@n5), B=ComponentMask(None=@n2) {b=true, g=true}, Alpha=NamedRerouteUsage())) {outputName=UVs}
  n6 = Abs(None=VertexNormalWS())
  n7 = ComponentMask(None=@n6) {g=true, r=true}
  n8 = ComponentMask(None=@n6) {b=true}
  n9 = Saturate(None=Subtract(A=@n7, B=@n8))
  n10 = ComponentMask(None=@n9) {r=true}
  n11 = ComponentMask(None=@n9) {g=true}
  n12 = Subtract(A=@n10, B=@n11)
  n13 = Add(A=@n12) {constB=0.5}
  n14 = NamedRerouteDeclaration(None=Saturate(None=Round(None=@n13))) {name=Flat R, nodeColor=<rgba 1 0 0.8906 1>}
  n15 = Add(A=Subtract(A=@n11, B=@n10)) {constB=0.5}
  n16 = NamedRerouteDeclaration(None=Saturate(None=Round(None=@n15))) {name=Flat G, nodeColor=<rgba 0 1 0.7891 1>}
  n17 = Max(A=@n14, B=@n16)
  FunctionOutput(None=AppendVector(A=AppendVector(A=@n14, B=@n16), B=OneMinus(None=@n17))) {outputName=Triplanar Mask, sortPriority=1}
  n18 = Subtract(A=@n8, B=@n7)
  n19 = ComponentMask(None=@n18) {r=true}
  n20 = ComponentMask(None=@n18) {g=true}
  n21 = Max(A=Saturate(None=Min(A=@n19, B=@n20)), B=Abs(None=@n12))
  n22 = Constant() {r=64}
  FunctionOutput(None=Power(
    Base=OneMinus(None=@n21)
    Exp=FunctionInput(Preview=@n22) {bUsePreviewValueAsDefault=true, inputName=Seam Sharpness, inputType=FunctionInput_Scalar, sortPriority=3}
   )) {outputName=Seam Mask, sortPriority=3}
  Round()
 ]}
