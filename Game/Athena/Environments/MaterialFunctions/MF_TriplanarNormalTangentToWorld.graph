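# Material function that converts a tangent-space normal map sample into world-space normals for triplanar mapping.
# For each axis plane it swizzles the tangent components, adds the in-plane pair to the vertex normal and multiplies Z by the axis component, a whiteout-style blend.
# X and Y planes negate tangent green; vertex normals default to VertexNormalWS.
# The Triplanar Mask channels weight the three results, which are summed without normalization.
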MaterialFunction /Game/Athena/Environments/MaterialFunctions/MF_TriplanarNormalTangentToWorld
graph {nodes [
  n1 = FunctionInput() {inputName=Tangent Normals}
  n2 = ComponentMask(None=@n1) {b=true}
  n3 = FunctionInput(Preview=VertexNormalWS()) {bUsePreviewValueAsDefault=true, inputName=Vertex Normals, sortPriority=1}
  n4 = ComponentMask(None=@n3) {r=true}
  n5 = ComponentMask(None=@n1) {r=true}
  n6 = ComponentMask(None=@n3) {g=true}
  n7 = ComponentMask(None=@n1) {g=true}
  n8 = Multiply(A=@n7) {constB=-1}
  n9 = ComponentMask(None=@n3) {b=true}
  n10 = FunctionInput() {inputName=Triplanar Mask, sortPriority=2}
  n11 = Multiply(
   A=MaterialFunctionCall(R=Multiply(A=@n2, B=@n4), G=Add(A=@n5, B=@n6), B=Add(A=@n8, B=@n9)) {materialFunction=&AppendMany}.RGB
   B=ComponentMask(None=@n10) {r=true}
  )
  n12 = Multiply(
   A=MaterialFunctionCall(R=Add(A=@n5, B=@n4), G=Multiply(A=@n2, B=@n6), B=Add(A=@n8, B=@n9)) {materialFunction=&AppendMany}.RGB
   B=ComponentMask(None=@n10) {g=true}
  )
  n13 = MaterialFunctionCall(R=Add(A=@n5, B=@n4), G=Add(A=@n7, B=@n6), B=Multiply(A=@n2, B=@n9)) {materialFunction=&AppendMany}
  n14 = ComponentMask(None=@n10) {b=true}
  FunctionOutput(None=Add(A=Add(A=@n11, B=@n12), B=Multiply(A=@n13.RGB, B=@n14))) {outputName=World Normals}
 ]}
